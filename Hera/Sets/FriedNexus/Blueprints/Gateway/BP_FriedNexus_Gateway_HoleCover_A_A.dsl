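; Gateway hole cover prop Blueprint for the FriedNexus set, registered as a simple destructible prop.
; Its construction script gets the SimpleDestructiblePropSubsystem and, if valid, converts the actor to simple destructible prop data.
; The conversion receives a one-element array holding the GlassProviderA variable.
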(fn ConstructionScript
  (bind _returnvalue (Class|SimpleDestructiblePropSubsystem|GetSimpleDestructiblePropSubsystem))
  (Utilities|IsValid _returnvalue
    (:Is%20Valid
      (Class|SimpleDestructiblePropSubsystem|ConstructionScriptConverttoSimpleDestructiblePropData _returnvalue self (Utilities|Array|MakeArray (Variables|Default|GetGlassProviderA) 0)))))
